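# Physics asset for the GroovyReaderPoke pickaxe, covering the rope bones of its axe mesh.
# Three small capsule bodies sit on base_rope, dyn_rope and dyn_rope_00.
# dyn_rope links to base_rope with a narrow and a wider swing limit and locked twist, while dyn_rope_00 is angularly locked to dyn_rope.
# An additional constraint joins base_rope directly to dyn_rope_00 with free swing and twist.
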PhysicsAsset /BRCosmetics/Weapons/FORT_Melee/Pickaxe_GroovyReaderPoke/Meshes/GroovyReaderPoke_Axe_Physics
bodies {
 base_rope [{shapeType=Capsule, length=0.505, radius=0.505, shapeName=base_rope_capsule, rotation=<rot 0 -90 0>}]
 dyn_rope [{shapeType=Capsule, length=2, radius=1, shapeName=dyn_rope_capsule, center=<5 0 0>, rotation=<rot 90 0 0>}]
 dyn_rope_00 [{shapeType=Capsule, length=2, radius=1, shapeName=dyn_rope_00_capsule, center=<9.7287 0 0>, rotation=<rot 90 0 0>}]
}
constraints [
 {
  bone1Name=dyn_rope_00
  bone2Name=dyn_rope
  swing1LimitDegrees=2
  swing1Motion=Locked
  swing2LimitDegrees=65
  swing2Motion=Locked
  twistLimitDegrees=10
  twistMotion=Locked
 }
 {bone1Name=dyn_rope, bone2Name=base_rope, swing1LimitDegrees=2, swing2LimitDegrees=40, twistLimitDegrees=0, twistMotion=Locked}
 {
  bone1Name=base_rope
  bone2Name=dyn_rope_00
  swing1LimitDegrees=45
  swing1Motion=Free
  swing2LimitDegrees=45
  swing2Motion=Free
  twistLimitDegrees=45
  twistMotion=Free
 }
]
