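# Animation Blueprint for the GroovyReaderPoke pickaxe that adds secondary physics motion to the weapon skeleton.
# A bool blend uses a single-bone Anim Dynamics spring on dyn_rope in the front end, and a RigidBody node with its own physics asset otherwise.
# Each update branches on player movement, but both paths copy Sim Space Settings_0, with its lower world alpha, into the RigidBody's sim-space settings.
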AnimBlueprint /BRCosmetics/Weapons/FORT_Melee/Pickaxe_GroovyReaderPoke/Meshes/GroovyReaderPoke_Axe_AnimBP
target_skeleton=&/GroovyReaderPoke_Axe_Skeleton
graphs [
 AnimGraph {nodes [
   n1 = Anim Dynamics - Bone: dyn_rope(ComponentPose=Local To Component(LocalPose=Input Pose() {name=InPose_0}), Alpha=1, ExternalForce="0,0,0") {
    angularDampingOverride=1
    angularSpringConstant=20000
    bAngularSpring=true
    bOverrideAngularDamping=true
    bUseGravityOverride=true
    gravityScale=0.35
    simulationSpace=RootRelative
    alphaBoolBlend {blendOption=Linear}
    boundBone {boneName=dyn_rope}
    componentLinearAccScale=<1 1 1>
    componentLinearVelScale=<1 1 1>
    gravityOverride {z=-980}
    physicsBodyDefinitions [{
      boundBone {boneName=dyn_rope}
      boxExtents=<20 30 30>
      constraintSetup {angularLimitsMax=<2 20 0>, angularLimitsMin=<-2 -20 0>, angularTarget=<1 0 0>}
      localJointOffset=<4 0 0>
     }]
   }
   n2 = RigidBody!(
    ComponentPose=Local To Component(LocalPose=Input Pose())
    Alpha
    ExternalForce
    SimSpaceSettings:Sim Space Settings Structure=Get Sim Space Settings() -> (Sim Space Settings:Sim Space Settings Structure)
   ) -> (Pose) {
    bClampLinearTranslationLimitToRefPose=true
    bOverrideWorldGravity=true
    overridePhysicsAsset=&/GroovyReaderPoke_Axe_Physics
    simulationSpace=BaseBoneSpace
    alphaBoolBlend {blendOption=Linear}
    baseBoneRef {boneName=b_AR_Root}
    componentAppliedLinearAccClamp=<3000 3000 3000>
    componentLinearAccScale=<0.6 0.35 0.6>
    componentLinearVelScale=<0.6 0.35 0.6>
    overrideWorldGravity {z=-600}
    simSpaceSettings {
     maxAngularAcceleration=1000
     maxAngularVelocity=3
     maxLinearAcceleration=700
     maxLinearVelocity=500
     velocityScaleZ=0
     worldAlpha=1
     externalLinearDragV=<0.2 0.2>
    }
   }
   Output Pose(Result=Blend Poses by bool(BlendPose_0=Component To Local(ComponentPose=@n1), BlendPose_1=Component To Local(ComponentPose=@n2), bActiveValue=Get bIsInFrontEnd()) {blendType=Linear}) {layerGroup=None}
  ]}
 EventGraph {nodes [
   Event BlueprintUpdateAnimation()
   > n1 = PlayerMovement(AnimInstance=Self-Reference(), XDotProductCoefficient=100, YDotProductCoefficient=100, ZDotProductCoefficient=100)
   n2 = MakeVector(X=@n1, Y=@n1.YMovement)
   > n3 = Branch(Condition="float < float"(A=BreakVector(InVec=@n2), B=-50))
   > Set Sim Space Settings(Sim Space Settings:Sim Space Settings Structure=Get Sim Space Settings_0() -> (Sim Space Settings_0:Sim Space Settings Structure)) -> (Output_Get:Sim Space Settings Structure)
   Set Sim Space Settings(execute=@n3.else, Sim Space Settings=Get Sim Space Settings_0())
   TryGetPawnOwner()
  ]}
]
variables [
 Backward {type=number}
 Forward {type=number}
 Left {type=number}
 Right {type=number}
 Sim Space Settings {
  type=SimSpaceSettings
  default {
   dampingAlpha=1
   maxAngularAcceleration=1000
   maxAngularVelocity=3
   maxLinearAcceleration=700
   maxLinearVelocity=500
   velocityScaleZ=0
   worldAlpha=1
   externalLinearDragV=<0.2 0.2 0>
  }
 }
 Sim Space Settings_0 {
  type=SimSpaceSettings
  default {
   dampingAlpha=1
   maxAngularAcceleration=1000
   maxAngularVelocity=3
   maxLinearAcceleration=700
   maxLinearVelocity=500
   velocityScaleZ=0
   worldAlpha=0.25
   externalLinearDragV=<0.2 0.2 0>
  }
 }
]
